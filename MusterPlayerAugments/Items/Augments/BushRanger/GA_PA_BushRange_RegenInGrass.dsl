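; Bush Ranger player augment ability for regeneration in stealth grass, triggered by a gameplay event.
; On activation it checks GetHealAllowed?, which passes if the GetHealableInStorm? scalable float is true or the avatar pawn is outside any storm.
; A passing check calls ActivateAugment.
; The ability also uses a WaitGameplayTagRemove task on the InStealthGrass tag to call EndAugmentEffect when that tag is removed.
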(event Ability|EventActivateAbilityFromEvent
  (params (EventData :type Gameplay%20Event%20Data%20Structure%20%28by%20ref%29))
  ;[1 0.2718 0.06097] Activate when equipping bush
  (if (CallFunction|GetHealAllowed?)
    (CallFunction|ActivateAugment)
    (Ability|Tasks|WaitGameplayTagRemove "(TagName=\"Gameplay.Action.Player.InStealthGrass\")"
      (:Removed
        (CallFunction|EndAugmentEffect)))))

(fn GetHealAllowed?
  ;[0.583 0.2655 1] Do not heal if inside the storm
  (return (select (ScalableFloat|GetValueasBool (Variables|Default|GetHealableInStorm?)) true (not (Variables|/Script/FortniteGame.FortPlayerPawn|GetbIsInAnyStorm (CallFunction|GetAvatarasFortPlayerPawn))))))
